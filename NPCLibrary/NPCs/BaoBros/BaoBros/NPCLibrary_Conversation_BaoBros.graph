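# Conversation database for the Bao Bros NPC in the NPC Library, defining its greeting and the services it leads to.
# The entry point feeds one speech node with a default line and contextual lines for a first-ever meeting or for players wearing specific character cosmetics, such as DumplingMan.
# That speech node then branches through link nodes into the shared NPC Library quests and basic services conversations.
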ConversationDatabase /NPCLibrary/NPCs/BaoBros/BaoBros/NPCLibrary_Conversation_BaoBros
entry_points [{tag=Conversation.Entry.BaoBros, destinations [#1]}]
nodes [
 ConversationEntryPointNode_7 {
  id=#1
  class=ConversationEntryPointNode
  outputs [#2]
  properties {entryTag {tagName=Conversation.Entry.BaoBros}}
 }
 Speech_C_4 {
  id=#2
  class=Speech_C
  outputs [#3, #4]
  properties {
   nodeName=""
   generalConfig {
    defaultMessage=It's not looking good out there.
    contextualMessages [
     {
      message=Taco Tuesday is cool and all but what about Dumpling Daturday?
      contextRequirements [{requirement {class=FortControllerRequirement_IsFirstEverConversationWithNPC}}]
      priority {curve {rowName=Tandem.Speech.Priority.FirstEverInteraction}}
     }
     {
      message=Your shift starts in ten minutes. Don't get eliminated.
      contextRequirements [{requirement {
         class=FortControllerRequirement_HasCID
         properties {allowedCIds [/BRCosmetics/Athena/Items/Cosmetics/Characters/CID_338_Athena_Commando_M_DumplingMan]}
        }}]
      priority {value=0, curve {curveTable=null, rowName=None}}
     }
     {
      message=You looking for a summer job?
      contextRequirements [{requirement {class=FortControllerRequirement_HasCID, properties {allowedCIds [&Character_ChillCat]}}}]
      priority {value=0, curve {curveTable=null, rowName=None}}
     }
     {
      message=How do you like my bunker?
      contextRequirements [{requirement {class=FortControllerRequirement_HasCID, properties {allowedCIds [&Character_Bites]}}}]
      priority {value=0, curve {curveTable=null, rowName=None}}
     }
     {
      message="You need a nice, hot meal?"
      contextRequirements [{requirement {class=FortControllerRequirement_HasCID, properties {allowedCIds [&Character_PinkSpike]}}}]
      priority {value=0, curve {curveTable=null, rowName=None}}
     }
     {
      message=What a very big suit for a very small bear.
      contextRequirements [{requirement {class=FortControllerRequirement_HasCID, properties {allowedCIds [&Character_BadBear]}}}]
      priority {value=0, curve {curveTable=null, rowName=None}}
     }
     {
      message=They got dumplings where you come from?
      contextRequirements [{requirement {class=FortControllerRequirement_HasCID, properties {allowedCIds [&Character_TheHerald]}}}]
      priority {value=0, curve {curveTable=null, rowName=None}}
     }
     {
      message=What exactly does a spider person eat?
      contextRequirements [{requirement {class=FortControllerRequirement_HasCID, properties {allowedCIds [&Character_RoseDust]}}}]
      priority {value=0, curve {curveTable=null, rowName=None}}
     }
    ]
   }
  }
 }
 ConversationLinkNode_0 {id=#4, class=ConversationLinkNode, properties {remoteEntryTag {tagName=Conversation.Entry.BasicServices}}}
 ConversationLinkNode_1 {id=#3, class=ConversationLinkNode, properties {remoteEntryTag {tagName=Conversation.Entry.NPCLibrary.Quests}}}
]
